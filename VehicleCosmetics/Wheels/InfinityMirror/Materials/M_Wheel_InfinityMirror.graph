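# Masked, lit material for the InfinityMirror vehicle wheel, built on MF_WheelParent_Simple and usable on skeletal meshes.
# Vertex-color blue, limited by inverted parent VertexColorR, selects where red and green tints from rim-mask channels replace base color.
# Within blue-painted areas, rim-mask blue alone drives opacity, cutting out the rest.
# The attributes then pass through the shared vehicle damage and de-rez transition functions.
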Material /VehicleCosmetics/Wheels/InfinityMirror/Materials/M_Wheel_InfinityMirror
graph {
 properties {
  bAutomaticallySetUsageInEditor=false
  bUseMaterialAttributes=true
  bUsedWithMeshDeformer=true
  bUsedWithSkeletalMesh=true
  blendMode=BLEND_Masked
  pixelDepthOffsetMode=PDOM_Legacy
  usedShadingModels=Default Lit
 }
 nodes [
  n1 = MaterialFunctionCall() {materialFunction=&MF_WheelParent_Simple}
  n2 = GetMaterialAttributes(None=@n1.MaterialAttributes) {attributeGetTypes [#1]}
  n3 = VertexColor()
  n4 = OneMinus(None=@n1.VertexColorR)
  n5 = Multiply(A=@n3.B, B=@n4)
  n6 = Add(A=@n2.Base Color, B=ComponentMask(None=@n1.RimRGBAMask) {r=true})
  n7 = Add(A=@n2.Base Color, B=ComponentMask(None=@n1.RimRGBAMask) {g=true})
  n8 = Add(
   A=Multiply(A=Constant3Vector() {constant=<rgba 1 0 0 0>}, B=Multiply(A=@n6, B=@n4))
   B=Multiply(A=Constant3Vector() {constant=<rgba 0 1 0 0>}, B=Multiply(A=@n7, B=@n4))
  )
  n9 = Multiply(A=Constant3Vector() {constant=<rgba 0.01 0.01 0.01 0>}, B=ComponentMask(None=@n1.RimRGBAMask) {b=true})
  n10 = Add(A=Multiply(A=@n2.Base Color, B=OneMinus(None=@n5)), B=Multiply(A=Add(A=@n8, B=@n9), B=@n5))
  n11 = Add(A=OneMinus(None=@n3.B), B=Multiply(A=ComponentMask(None=@n1.RimRGBAMask) {b=true}, B=@n3.B))
  MaterialFunctionCall(In=MaterialFunctionCall(Material=SetMaterialAttributes!(MaterialAttributes=@n2.MaterialAttributes, Base Color=@n10, Opacity Mask=@n11) {attributeSetTypes [#1, #2]}) {materialFunction=&MF_BaseVehicle_Damage}.Result) {materialFunction=&MF_VehicleDeRez_Transition}
 ]
}
